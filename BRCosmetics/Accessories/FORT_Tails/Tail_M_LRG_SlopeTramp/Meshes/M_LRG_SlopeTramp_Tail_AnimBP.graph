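# Animation Blueprint for the SlopeTramp tail cosmetic that re-poses the tail base during knee slides, baseball slides and sprinting.
# The event graph eases one alpha per movement state toward on or off with FInterpTo, treating regular and tactical sprint alike.
# These alphas drive chained additive Transform Bone nodes on the base joint, with a stronger yaw for slides than sprints.
# A final always-on node applies a small constant translation offset.
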AnimBlueprint /BRCosmetics/Accessories/FORT_Tails/Tail_M_LRG_SlopeTramp/Meshes/M_LRG_SlopeTramp_Tail_AnimBP
target_skeleton=/Game/Accessories/FORT_Tails/Common/Fortnite_Base_Tail_Skeleton
graphs [
 AnimGraph {nodes [
   n1 = "Transform (Modify) Bone - Bone: C_Tail_A_Base_Jnt"(
    ComponentPose=Local To Component(LocalPose=Input Pose())
    Alpha=Get tail rot() -> (tail rot:double)
    Translation="0,0,-10"
    Rotation="0,121,0"
    Scale="1,1,1"
   ) {
    rotationMode=BMM_Additive
    rotationSpace=BCS_BoneSpace
    translationMode=BMM_Additive
    alphaBoolBlend {blendOption=Linear}
    boneToModify {boneName=C_Tail_A_Base_Jnt}
    rotation {yaw=121}
    translation {z=-10}
   }
   n2 = "Transform (Modify) Bone - Bone: C_Tail_A_Base_Jnt"(ComponentPose=@n1, Alpha=Get tail rot_2() -> (tail rot_2:double), Translation="0,0,-10", Rotation="0,121,0", Scale="1,1,1") {
    rotationMode=BMM_Additive
    rotationSpace=BCS_BoneSpace
    translationMode=BMM_Additive
    alphaBoolBlend {blendOption=Linear}
    boneToModify {boneName=C_Tail_A_Base_Jnt}
    rotation {yaw=121}
    translation {z=-10}
   }
   n3 = "Transform (Modify) Bone - Bone: C_Tail_A_Base_Jnt"(ComponentPose=@n2, Alpha=Get tail rot_3() -> (tail rot_3:double), Translation="0,0,-7", Rotation="0,35,0", Scale="1,1,1") {
    rotationMode=BMM_Additive
    rotationSpace=BCS_BoneSpace
    translationMode=BMM_Additive
    alphaBoolBlend {blendOption=Linear}
    boneToModify {boneName=C_Tail_A_Base_Jnt}
    rotation {yaw=35}
    translation {z=-7}
   }
   n4 = "Transform (Modify) Bone - Bone: C_Tail_A_Base_Jnt"(ComponentPose=@n3, Alpha=1, Translation="0,-1.7,0", Rotation="0,0,0", Scale="1,1,1") {
    rotationMode=BMM_Additive
    translationMode=BMM_Additive
    translationSpace=BCS_BoneSpace
    alphaBoolBlend {blendOption=Linear}
    boneToModify {boneName=C_Tail_A_Base_Jnt}
    translation {y=-1.7}
   }
   Output Pose(Result=Component To Local(ComponentPose=@n4))
  ]}
 EventGraph {nodes [
   n1 = Event BlueprintUpdateAnimation()
   > Set Delta Time X(Delta Time X=@n1.DeltaTimeX)
   > n2 = Sequence!(execute) -> (then_0, then_1, then_2:Exec)
   > n3 = Branch(Condition=Get bIsKneeSliding())
   > Set tail rot(tail rot:double=FInterpTo(Current=Get tail rot(), Target=1, DeltaTime=Get Delta Time X(), InterpSpeed=5)) -> (Output_Get:double)
   Set tail rot(execute=@n3.else, tail rot=FInterpTo(Current=Get tail rot(), Target=0, DeltaTime=Get Delta Time X(), InterpSpeed=5))
   n4 = Branch(execute=@n2.then_1, Condition=Get bIsBaseballSliding())
   > Set tail rot_2(tail rot_2:double=FInterpTo(Current=Get tail rot_2(), Target=1, DeltaTime=Get Delta Time X(), InterpSpeed=5)) -> (Output_Get:double)
   Set tail rot_2(execute=@n4.else, tail rot_2=FInterpTo(Current=Get tail rot_2(), Target=0, DeltaTime=Get Delta Time X(), InterpSpeed=5))
   n5 = Branch(execute=@n2.then_2, Condition=OR Boolean(A=Get bIsSprinting(), B=Get bIsTacticalSprinting()))
   > Set tail rot_3(tail rot_3:double=FInterpTo(Current=Get tail rot_3(), Target=1, DeltaTime=Get Delta Time X(), InterpSpeed=5)) -> (Output_Get:double)
   Set tail rot_3(execute=@n5.else, tail rot_3=FInterpTo(Current=Get tail rot_3(), Target=0, DeltaTime=Get Delta Time X(), InterpSpeed=5))
   TryGetPawnOwner()
  ]}
]
variables [
 Delta Time X {type=number}
 tail rot {type=number}
 tail rot_2 {type=number}
 tail rot_3 {type=number}
]
